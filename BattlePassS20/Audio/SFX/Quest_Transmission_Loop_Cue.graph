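# Quest_Transmission_Loop_Cue is a Season 20 Battle Pass sound cue for a looping quest transmission sound effect.
# A single looping wave player plays the Quest_Transmission_Loop wave directly as the cue root.
# It routes through the Fort_Quests sound class and uses the HiddenStar ambient loop attenuation from SupplyDrops for spatial falloff.
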SoundCue /BattlePassS20/Audio/SFX/Quest_Transmission_Loop_Cue
root=WavePlayer_0
nodes [{id=WavePlayer_0, properties {bLooping=true, soundWaveAssetPtr=&/Quest_Transmission_Loop}}]
properties {
 attenuationSettings=/Game/Athena/Sounds/SupplyDrops/Fort_HiddenStar_Ambient_Loop_Att
 maxDistance=2500
 soundClassObject=&Fort_Quests
 volumeMultiplier=1
}
