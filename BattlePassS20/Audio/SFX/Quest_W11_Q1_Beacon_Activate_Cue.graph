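# Sound cue for a beacon activation effect, named for a Battle Pass Season 20 week 11 quest.
# A single wave player feeds the Quest_W11_Q1_Beacon_Activate sound wave straight to the output, with no randomization, mixing or modulation.
# It uses the non-critical in-game UI sound class at slightly reduced volume.
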SoundCue /BattlePassS20/Audio/SFX/Quest_W11_Q1_Beacon_Activate_Cue
root=WavePlayer_0
nodes [{id=WavePlayer_0, properties {soundWaveAssetPtr=&/Quest_W11_Q1_Beacon_Activate}}]
properties {maxDistance=2097152, soundClassObject=&Fort_UI_InGame_NonCritical, volumeMultiplier=0.8}
