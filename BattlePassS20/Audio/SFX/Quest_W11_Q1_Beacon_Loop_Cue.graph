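# Sound cue for a looping beacon ambience tied to a Battle Pass Season 20, Week 11 quest.
# A single looping wave player feeds the Quest_W11_Q1_Beacon_Loop wave straight to the output, with a boosted volume multiplier.
# It routes through the Fort_Quests sound class and borrows the supply-drop Hidden Star ambient loop attenuation for spatial falloff.
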SoundCue /BattlePassS20/Audio/SFX/Quest_W11_Q1_Beacon_Loop_Cue
root=WavePlayer_0
nodes [{id=WavePlayer_0, properties {bLooping=true, soundWaveAssetPtr=&/Quest_W11_Q1_Beacon_Loop}}]
properties {
 attenuationSettings=/Game/Athena/Sounds/SupplyDrops/Fort_HiddenStar_Ambient_Loop_Att
 maxDistance=2500
 soundClassObject=&Fort_Quests
 volumeMultiplier=1.5
}
