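# Sound cue for a quest target hit in the Season 20 Battle Pass content, playing an impact sound.
# A Random node picks one of three Quest_TargetHit wave variations with equal weighting.
# The cue routes to the Fort_Impacts sound class with a boosted volume and a very large max distance.
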SoundCue /BattlePassS20/Audio/SFX/TargetHit_Cue
root=Random_1
nodes [
 {id=Random_1, children [WavePlayer_0,WavePlayer_1,WavePlayer_2], properties {weights [1,1,1]}}
 {id=WavePlayer_0, properties {soundWaveAssetPtr=&/Quest_TargetHit_03}}
 {id=WavePlayer_1, properties {soundWaveAssetPtr=&/Quest_TargetHit_01}}
 {id=WavePlayer_2, properties {soundWaveAssetPtr=&/Quest_TargetHit_02}}
]
properties {maxDistance=2097152, soundClassObject=&Fort_Impacts, volumeMultiplier=1.3667}
